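# Control Rig for the BasilStrong Mashup face accessory that pushes the neck scarf bones as the head rotates, likely to avoid clipping.
# Three spherical pose readers measure the head bone against spine_05, each with its own rotated region and falloff, with debug drawing left on.
# Their outputs weight a shared translation offset on the right, left and center neck scarf bones, propagated to children.
# Other scarf and deform bones in the hierarchy are not driven.
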ControlRigBlueprint /BRCosmetics/Characters/Player/Male/Large/Bodies/M_LRG_BasilStrong_Mashup/Meshes/Parts/M_LRG_BasilStrong_Mashup_FaceAcc_CtrlRig
graphs [
 RigVMFunctionLibrary {links [], nodes []}
 RigVMModel {nodes [
   RigUnit_BeginExecution = &RigUnit_BeginExecution()
   n1 = &RigUnit_SphericalPoseReader(
    ActiveRegionScaleFactors=(PositiveWidth=0.01,NegativeWidth=0.01,PositiveHeight=0.01,NegativeHeight=0.01)
    ActiveRegionSize=0.5
    Debug=(bDrawDebug=True,DebugScale=25,DebugSegments=20,DebugThickness=0.25)
    DriverAxis=(X=0,Y=1,Z=0)
    DriverItem=(Type=Bone,Name="head")
    FalloffRegionScaleFactors=(PositiveWidth=0.4,NegativeWidth=0.4,PositiveHeight=0.2,NegativeHeight=0.2)
    FalloffSize=1
    OptionalParentItem=(Type=Bone,Name="spine_05")
    RotationOffset=(X=45,Y=0,Z=0)
    ExecutePin=@RigUnit_BeginExecution.ExecutePin
   )
   > n2 = &RigUnit_SphericalPoseReader(
    ActiveRegionScaleFactors=(PositiveWidth=0.01,NegativeWidth=0.01,PositiveHeight=0.01,NegativeHeight=0.01)
    ActiveRegionSize=0.5
    Debug=(bDrawDebug=True,DebugScale=25,DebugSegments=20,DebugThickness=0.25)
    DriverAxis=(X=0,Y=1,Z=0)
    DriverItem=(Type=Bone,Name="head")
    FalloffRegionScaleFactors=(PositiveWidth=0.4,NegativeWidth=0.4,PositiveHeight=0.3,NegativeHeight=0.2)
    FalloffSize=1
    OptionalParentItem=(Type=Bone,Name="spine_05")
    RotationOffset=(X=135,Y=0,Z=0)
   )
   > n3 = &RigUnit_SphericalPoseReader(
    ActiveRegionScaleFactors=(PositiveWidth=0.01,NegativeWidth=0.01,PositiveHeight=0.01,NegativeHeight=0.01)
    ActiveRegionSize=0.5
    Debug=(bDrawDebug=True,DebugScale=25,DebugSegments=20,DebugThickness=0.25)
    DriverAxis=(X=0,Y=1,Z=0)
    DriverItem=(Type=Bone,Name="head")
    FalloffRegionScaleFactors=(PositiveWidth=0.15,NegativeWidth=0.4,PositiveHeight=0.4,NegativeHeight=0.4)
    FalloffSize=1
    OptionalParentItem=(Type=Bone,Name="spine_05")
    RotationOffset=(X=90,Y=0,Z=35)
   )
   > &RigUnit_OffsetTransformForItem(
    Item=(Type=Bone,Name="dyn_Neck_scarf_r")
    OffsetTransform=(Rotation=(X=0,Y=0,Z=0,W=1),Translation=(X=-5,Y=10,Z=0),Scale3D=(X=1,Y=1,Z=1))
    bPropagateToChildren=True
    Weight=@n1.OutputParam
   )
   > &RigUnit_OffsetTransformForItem(
    Item=(Type=Bone,Name="dyn_neck_scarf_l")
    OffsetTransform=(Rotation=(X=0,Y=0,Z=0,W=1),Translation=(X=-5,Y=10,Z=0),Scale3D=(X=1,Y=1,Z=1))
    bPropagateToChildren=True
    Weight=@n2.OutputParam
   )
   > &RigUnit_OffsetTransformForItem(
    Item=(Type=Bone,Name="dyn_neck_scarf_c")
    OffsetTransform=(Rotation=(X=0,Y=0,Z=0,W=1),Translation=(X=-5,Y=10,Z=0),Scale3D=(X=1,Y=1,Z=1))
    bPropagateToChildren=True
    Weight=@n3.OutputParam
   )
  ]}
]
hierarchy [
 root {loc=<0 0 0>, rot=<0 0 0 1>}
 pelvis {parent=root, loc=<0 -0.7321 92.4218>, rot=<0.0127 0.707 -0.0127 -0.707>}
 spine_01 {parent=pelvis, loc=<5.6002 -0.5128 0>, rot=<0 0 0.034 -0.9994>}
 spine_02 {parent=spine_01, loc=<8.0753 0.0139 0>, rot=<0 0 0.0238 -0.9997>}
 spine_03 {parent=spine_02, loc=<7.0393 0 0>, rot=<0 0 0.0149 -0.9999>}
 spine_04 {parent=spine_03, loc=<8.1024 0 0>, rot=<0 0 -0.0132 -0.9999>}
 spine_05 {parent=spine_04, loc=<10.0146 0 0>, rot=<0 0 -0.0015 -1>}
 clavicle_l {parent=spine_05, loc=<2.5856 -0.2866 -4.3777>, rot=<0.0465 -0.7289 -0.009 -0.6829>}
 upperarm_l {parent=clavicle_l, loc=<18.2264 0.1342 -0.001>, rot=<0.0253 -0.2693 0.0013 -0.9627>}
 upperarm_twist_01_l {parent=upperarm_l, loc=<7.9796 0.0843 -0.6195>, rot=<0 0 0 1>}
 deform_upperarm_bk_base_l {parent=upperarm_twist_01_l, loc=<-6.8983 -13.3826 -1.3258>, rot=<0 0 0 -1>}
 deform_upperarm_bk_l {parent=deform_upperarm_bk_base_l, loc=<0 0 0>, rot=<0 0 0 1>}
 deform_upperarm_fr_base_l {parent=upperarm_twist_01_l, loc=<-8.3487 10.7117 -2.1455>, rot=<0 0 0 -1>}
 deform_upperarm_fr_l {parent=deform_upperarm_fr_base_l, loc=<0 0 0>, rot=<0 0 0 1>}
 deform_upperarm_up_base_l {parent=upperarm_twist_01_l, loc=<-11.2012 -1.2234 8.7784>, rot=<0 0 0 -1>}
 deform_upperarm_up_l {parent=deform_upperarm_up_base_l, loc=<0 0 0>, rot=<0 0 0 1>}
 upperarm_twist_02_l {parent=upperarm_l, loc=<19.5059 0.0843 -0.6195>, rot=<0 0 0 1>}
 neck_01 {parent=spine_05, loc=<10.815 1.9196 0>, rot=<0 0 -0.1864 -0.9825>}
 neck_02 {parent=neck_01, loc=<4.7347 0.3045 0>, rot=<0 0 0.0363 -0.9993>}
 head {parent=neck_02, loc=<5.7632 0.5552 0>, rot=<0 0 0.115 -0.9934>}
 dyn_neck_scarf_l_base {parent=neck_02, loc=<0 0 0>, rot=<0.4444 0 0 -0.8959>}
 dyn_neck_scarf_l {parent=dyn_neck_scarf_l_base, loc=<0 10.3181 0>, rot=<0 0 0 1>}
 dyn_neck_scarf_c_base {parent=neck_02, loc=<0 0 0>, rot=<0 0 0 1>}
 dyn_neck_scarf_c {parent=dyn_neck_scarf_c_base, loc=<0 13.856 0>, rot=<0 0 0 1>}
 dyn_neck_scarf_r_base {parent=neck_02, loc=<0 0 0>, rot=<-0.3918 0 0 -0.92>}
 dyn_Neck_scarf_r {parent=dyn_neck_scarf_r_base, loc=<0 11.1845 0>, rot=<0 0 0 1>}
 clavicle_r {parent=spine_05, loc=<2.5855 -0.2866 4.3777>, rot=<0.7289 0.0465 -0.6829 0.009>}
 upperarm_r {parent=clavicle_r, loc=<-18.2264 -0.1342 0.001>, rot=<0.0253 -0.2693 0.0013 -0.9627>}
 upperarm_twist_01_r {parent=upperarm_r, loc=<-7.9796 -0.0843 0.6195>, rot=<0 0 0 1>}
 deform_upperarm_bk_base_r {parent=upperarm_twist_01_r, loc=<6.8985 13.3826 1.3257>, rot=<0 0 0 -1>}
 deform_upperarm_bk_r {parent=deform_upperarm_bk_base_r, loc=<0 0 0>, rot=<0 0 0 1>}
 deform_upperarm_fr_base_r {parent=upperarm_twist_01_r, loc=<8.3485 -10.7117 2.1456>, rot=<0 0 0 -1>}
 deform_upperarm_fr_r {parent=deform_upperarm_fr_base_r, loc=<0 0 0>, rot=<0 0 0 1>}
 deform_upperarm_up_base_r {parent=upperarm_twist_01_r, loc=<11.2015 1.2234 -8.7788>, rot=<0 0 0 -1>}
 deform_upperarm_up_r {parent=deform_upperarm_up_base_r, loc=<0 0 0>, rot=<0 0 0 1>}
 upperarm_twist_02_r {parent=upperarm_r, loc=<-19.5059 -0.0843 0.6195>, rot=<0 0 0 1>}
 dyn_scarfBack_base {parent=spine_05, loc=<18.6154 -4.5056 0>, rot=<0.5134 -0.8581 0 0>}
 dyn_scarfBack_01 {parent=dyn_scarfBack_base, loc=<0 0 0>, rot=<0 0 0 1>}
 dyn_scarfBack_02 {parent=dyn_scarfBack_01, loc=<6.0321 0 0>, rot=<0 0 -0.1736 -0.9848>}
 dyn_scarfBack_03 {parent=dyn_scarfBack_02, loc=<5.713 0 0>, rot=<0 0 0 1>}
 dyn_scarf_base {parent=spine_05, loc=<4.8567 17.6787 0>, rot=<0 0 0.9511 0.3088>}
 dyn_scarf_01 {parent=dyn_scarf_base, loc=<0 0 0>, rot=<0 0 0 1>}
 dyn_scarf_02 {parent=dyn_scarf_01, loc=<6.0321 0 0>, rot=<0 0 -0.1736 -0.9848>}
 dyn_scarf_03 {parent=dyn_scarf_02, loc=<5.713 0 0>, rot=<0 0 -0.0872 -0.9962>}
 deform_pec_aim_base_l {parent=spine_04, loc=<5.0232 13.6719 -2.6508>, rot=<0.0672 -0.5201 0.3228 -0.7879>}
 deform_pec_aim_l {parent=deform_pec_aim_base_l, loc=<0 0 0>, rot=<0 0 0 1>}
 deform_pec_base_l {parent=deform_pec_aim_l, loc=<16.3279 0 0>, rot=<0 0 0 1>}
 deform_pec_l {parent=deform_pec_base_l, loc=<0 0 0>, rot=<0 0 0 1>}
 deform_pec_aim_base_r {parent=spine_04, loc=<5.023 13.6718 2.6508>, rot=<-0.5201 -0.0672 0.7879 0.3228>}
 deform_pec_aim_r {parent=deform_pec_aim_base_r, loc=<0 0 0>, rot=<0 0 0 1>}
 deform_pec_base_r {parent=deform_pec_aim_r, loc=<-16.3279 0 0>, rot=<0 0 0 -1>}
 deform_pec_r {parent=deform_pec_base_r, loc=<0 0 0.0002>, rot=<0 0 0 -1>}
 deform_lat_l_base {parent=spine_03, loc=<2.5566 -10.3716 -17.6918>, rot=<0 0 0 1>}
 deform_lat_l {parent=deform_lat_l_base, loc=<0 0 0>, rot=<0 0 0 1>}
 deform_lat_r_base {parent=spine_03, loc=<2.5568 -10.3716 17.6918>, rot=<0 0 1 0>}
 deform_lat_r {parent=deform_lat_r_base, loc=<0 0 0>, rot=<0 0 0 -1>}
]
