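# Physics asset for the face accessory of the large male BasilStrong Mashup cosmetic body, covering a scarf with front and back dangling parts.
# Each part is a three-bone capsule chain hanging from a tiny base body constrained to spine_05, with larger spine_04 and spine_05 capsules representing the torso.
# Constraints along the chains limit swing unevenly per axis and allow zero twist, while base-to-spine and spine-to-spine constraints are free in rotation.
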PhysicsAsset /BRCosmetics/Characters/Player/Male/Large/Bodies/M_LRG_BasilStrong_Mashup/Meshes/Parts/M_LRG_BasilStrong_Mashup_FaceAcc_Physics
bodies {
 dyn_scarfBack_01 [{shapeType=Capsule, length=20, radius=4, shapeName=dyn_scarfBack_01_capsule, center=<3 -4 0>}]
 dyn_scarfBack_02 [{shapeType=Capsule, length=17, radius=4, center=<2 -1.5069 0>}]
 dyn_scarfBack_03 [{shapeType=Capsule, length=8, radius=3, shapeName=dyn_scarfBack_03_capsule, center=<2 -1.6494 0>}]
 dyn_scarfBack_base [{shapeType=Capsule, length=0.505, radius=0.505, shapeName=dyn_scarfBack_base_capsule, rotation=<rot 0 -90 0>}]
 dyn_scarf_01 [{shapeType=Capsule, length=18, radius=4, shapeName=dyn_scarf_01_capsule, center=<4.1558 -0.6313 0>}]
 dyn_scarf_02 [{shapeType=Capsule, length=15, radius=4, center=<3.5081 0.0535 0>}]
 dyn_scarf_03 [{shapeType=Capsule, length=10, radius=4, shapeName=dyn_scarf_03_capsule, center=<3.8149 -0.2991 0>}]
 dyn_scarf_base [{shapeType=Capsule, length=0.505, radius=0.505, shapeName=dyn_scarf_base_capsule, rotation=<rot 0 -90 0>}]
 spine_04 [{shapeType=Capsule, length=8, radius=18, shapeName=spine_04_capsule, center=<2.9112 0 0>, rotation=<rot 90 0 0>}]
 spine_05 [{shapeType=Capsule, length=19.2109, radius=12, center=<4.7515 -1.7107 -1.2674>, rotation=<rot 0.3739 -90 -1.3884>}]
}
constraints [
 {bone1Name=dyn_scarf_02, bone2Name=dyn_scarf_01, swing1LimitDegrees=35, swing2LimitDegrees=10, twistLimitDegrees=0}
 {bone1Name=dyn_scarfBack_02, bone2Name=dyn_scarfBack_01, swing1LimitDegrees=30, swing2LimitDegrees=20, twistLimitDegrees=0}
 {
  bone1Name=dyn_scarf_base
  bone2Name=spine_05
  swing1LimitDegrees=45
  swing1Motion=Free
  swing2LimitDegrees=45
  swing2Motion=Free
  twistLimitDegrees=45
  twistMotion=Free
 }
 {bone1Name=dyn_scarf_01, bone2Name=dyn_scarf_base, swing1LimitDegrees=50, swing2LimitDegrees=5, twistLimitDegrees=0}
 {bone1Name=dyn_scarf_03, bone2Name=dyn_scarf_02, swing1LimitDegrees=35, swing2LimitDegrees=15, twistLimitDegrees=0}
 {
  bone1Name=spine_05
  bone2Name=spine_04
  swing1LimitDegrees=45
  swing1Motion=Free
  swing2LimitDegrees=45
  swing2Motion=Free
  twistLimitDegrees=45
  twistMotion=Free
 }
 {bone1Name=dyn_scarfBack_01, bone2Name=dyn_scarfBack_base, swing1LimitDegrees=30, swing2LimitDegrees=10, twistLimitDegrees=0}
 {bone1Name=dyn_scarfBack_03, bone2Name=dyn_scarfBack_02, swing1LimitDegrees=20, swing2LimitDegrees=20, twistLimitDegrees=0}
 {
  bone1Name=dyn_scarfBack_base
  bone2Name=spine_05
  swing1LimitDegrees=45
  swing1Motion=Free
  swing2LimitDegrees=45
  swing2Motion=Free
  twistLimitDegrees=45
  twistMotion=Free
 }
]
